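# Sound cue for the FancyCandyMale one-handed pickaxe hitting an enemy, with melee impact attenuation and concurrency.
# A quality-level root selects between branches with differing variant pools, reusing impact waves from the FancyCandyPlainMale set.
# Most branches make a weighted pick between a modulated impact alone and, less often, that impact mixed with a quieter random vox line.
# One reduced branch plays only a briefly delayed impact without vox.
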SoundCue /BRCosmetics/Sounds/PickAxes/FancyCandyMale_1H/PickaxeImpactEnemy_FancyCandyMale_1H_Cue
root=QualityLevel_0
nodes [
 {id=Delay_3, children [Modulator_3], properties {delayMax=0.05, delayMin=0.05}}
 {id=Mixer_0, children [Modulator_1,Modulator_0], properties {inputVolume [1,0.6]}}
 {id=Mixer_1, children [Modulator_2,Modulator_4], properties {inputVolume [1,0.6]}}
 {id=Modulator_0, children [Random_1]}
 {id=Modulator_1, children [Random_2]}
 {id=Modulator_2, children [Random_3]}
 {id=Modulator_3, children [Random_4]}
 {id=Modulator_4, children [Random_9]}
 {id=QualityLevel_0, children [Random_5,Random_7,Delay_3,Random_7]}
 {
  id=Random_1
  children [WavePlayer_11, WavePlayer_26, WavePlayer_27, WavePlayer_28, WavePlayer_29, WavePlayer_30, WavePlayer_31]
  properties {weights [1,1,1,1,1,1,1]}
 }
 {
  id=Random_2
  children [WavePlayer_1, WavePlayer_2, WavePlayer_3, WavePlayer_4, WavePlayer_5, WavePlayer_12]
  properties {weights [1,1,1,1,1,1]}
 }
 {id=Random_3, children [WavePlayer_13, WavePlayer_14, WavePlayer_15], properties {weights [1,1,1]}}
 {id=Random_4, children [WavePlayer_16, WavePlayer_17], properties {weights [1,1]}}
 {id=Random_5, children [Modulator_1,Mixer_0], properties {bRandomizeWithoutReplacement=false, weights [1,0.4]}}
 {id=Random_7, children [Modulator_2,Mixer_1], properties {bRandomizeWithoutReplacement=false, weights [1,0.4]}}
 {id=Random_9, children [WavePlayer_35, WavePlayer_36, WavePlayer_37], properties {weights [1,1,1]}}
 {
  id=WavePlayer_1
  properties {soundWaveAssetPtr=/BRCosmetics/Sounds/PickAxes/FancyCandyPlainMale_1H/PickAxe_FancyCandyMalePlain_EnemyImpact_01}
 }
 {id=WavePlayer_11, properties {soundWaveAssetPtr=&/PickAxe_FancyCandyMale_Impact_Vox_01}}
 {
  id=WavePlayer_12
  properties {soundWaveAssetPtr=/BRCosmetics/Sounds/PickAxes/FancyCandyPlainMale_1H/PickAxe_FancyCandyMalePlain_EnemyImpact_06}
 }
 {
  id=WavePlayer_13
  properties {soundWaveAssetPtr=&PickAxe_FancyCandyMalePlain_EnemyImpact_01}
 }
 {
  id=WavePlayer_14
  properties {soundWaveAssetPtr=/BRCosmetics/Sounds/PickAxes/FancyCandyPlainMale_1H/PickAxe_FancyCandyMalePlain_EnemyImpact_03}
 }
 {
  id=WavePlayer_15
  properties {soundWaveAssetPtr=/BRCosmetics/Sounds/PickAxes/FancyCandyPlainMale_1H/PickAxe_FancyCandyMalePlain_EnemyImpact_02}
 }
 {
  id=WavePlayer_16
  properties {soundWaveAssetPtr=&PickAxe_FancyCandyMalePlain_EnemyImpact_01}
 }
 {
  id=WavePlayer_17
  properties {soundWaveAssetPtr=&PickAxe_FancyCandyMalePlain_EnemyImpact_02}
 }
 {
  id=WavePlayer_2
  properties {soundWaveAssetPtr=&PickAxe_FancyCandyMalePlain_EnemyImpact_02}
 }
 {id=WavePlayer_26, properties {soundWaveAssetPtr=&/PickAxe_FancyCandyMale_Impact_Vox_02}}
 {id=WavePlayer_27, properties {soundWaveAssetPtr=&/PickAxe_FancyCandyMale_Impact_Vox_03}}
 {id=WavePlayer_28, properties {soundWaveAssetPtr=&/PickAxe_FancyCandyMale_Impact_Vox_04}}
 {id=WavePlayer_29, properties {soundWaveAssetPtr=&/PickAxe_FancyCandyMale_Impact_Vox_05}}
 {
  id=WavePlayer_3
  properties {soundWaveAssetPtr=&PickAxe_FancyCandyMalePlain_EnemyImpact_03}
 }
 {id=WavePlayer_30, properties {soundWaveAssetPtr=&/PickAxe_FancyCandyMale_Impact_Vox_06}}
 {id=WavePlayer_31, properties {soundWaveAssetPtr=&/PickAxe_FancyCandyMale_Impact_Vox_07}}
 {id=WavePlayer_35, properties {soundWaveAssetPtr=&PickAxe_FancyCandyMale_Impact_Vox_01}}
 {id=WavePlayer_36, properties {soundWaveAssetPtr=&PickAxe_FancyCandyMale_Impact_Vox_02}}
 {id=WavePlayer_37, properties {soundWaveAssetPtr=&PickAxe_FancyCandyMale_Impact_Vox_03}}
 {
  id=WavePlayer_4
  properties {soundWaveAssetPtr=/BRCosmetics/Sounds/PickAxes/FancyCandyPlainMale_1H/PickAxe_FancyCandyMalePlain_EnemyImpact_04}
 }
 {
  id=WavePlayer_5
  properties {soundWaveAssetPtr=/BRCosmetics/Sounds/PickAxes/FancyCandyPlainMale_1H/PickAxe_FancyCandyMalePlain_EnemyImpact_05}
 }
]
properties {
 attenuationSettings=/Game/Sounds/Attenuation/Fort_Pickaxe_Impact_Att
 audioPropertiesSheet=/BRCosmetics/Sounds/Settings/PropertySheets/Pickaxe/APS_Cosmetic_Pickaxe_Impact
 bPrimeOnLoad=true
 maxDistance=4300
 soundClassObject=&Fort_WeaponMelee_PSM
 volumeMultiplier=0.5
 concurrencySet [&Player_Melee_Weapon_Conc]
}
