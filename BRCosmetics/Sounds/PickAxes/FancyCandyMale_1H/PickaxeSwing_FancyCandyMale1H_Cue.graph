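# Sound cue for the FancyCandyMale one-handed pickaxe swing cosmetic, limited to two voices per owner, stopping the oldest.
# A quality switch gives the top tier mixed close and medium-attenuation layers, the latter using generic distant fire-axe swings.
# Lower tiers use only the close attenuation with fewer variations, the last two sharing one branch.
# Weighted randoms occasionally layer vocal clips over swings, slightly delayed in the lower tiers.
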SoundCue /BRCosmetics/Sounds/PickAxes/FancyCandyMale_1H/PickaxeSwing_FancyCandyMale1H_Cue
root=QualityLevel_1
nodes [
 {id=Attenuation_1, children [Modulator_0], properties {attenuationSettings=/Game/Sounds/Attenuation/Fort_Weapon_Pickaxe_Close_Att}}
 {id=Attenuation_2, children [Modulator_1], properties {attenuationSettings=&Fort_Weapon_Pickaxe_Close_Att}}
 {id=Attenuation_6, children [Random_3], properties {attenuationSettings=&Fort_Weapon_Pickaxe_Close_Att}}
 {id=Attenuation_7, children [Modulator_8], properties {attenuationSettings=/Game/Sounds/Attenuation/Fort_Weapon_Pickaxe_Medium_Att}}
 {id=Delay_2, children [Random_10], properties {delayMax=0.05, delayMin=0.035}}
 {id=Delay_3, children [Random_13], properties {delayMax=0.05, delayMin=0.035}}
 {id=Mixer_0, children [Random_1,Random_2], properties {inputVolume [1,0.8]}}
 {id=Mixer_2, children [Random_0,Random_8], properties {inputVolume [1,0.1]}}
 {id=Mixer_3, children [Attenuation_6, Attenuation_7], properties {inputVolume [1,1]}}
 {id=Mixer_6, children [Random_9,Delay_2], properties {inputVolume [1,0.7]}}
 {id=Mixer_7, children [Random_12,Delay_3], properties {inputVolume [1,0.7]}}
 {id=Mixer_8, children [Random_1], properties {inputVolume [0.8]}}
 {id=Modulator_0, children [Random_11], properties {pitchMax=1.1, pitchMin=1, volumeMax=0.35, volumeMin=0.35}}
 {id=Modulator_1, children [Random_14], properties {pitchMax=1.1, pitchMin=1, volumeMax=0.35, volumeMin=0.35}}
 {id=Modulator_8, children [Random_4], properties {pitchMax=1.1, pitchMin=1, volumeMax=0.35, volumeMin=0.35}}
 {id=QualityLevel_1, children [Mixer_3, Attenuation_1, Attenuation_2, Attenuation_2]}
 {id=Random_0, children [WavePlayer_4, WavePlayer_5, WavePlayer_12], properties {weights [1,1,1]}}
 {id=Random_1, children [WavePlayer_0,WavePlayer_1,WavePlayer_3,WavePlayer_6,WavePlayer_7,WavePlayer_8], properties {weights [1,1,1,1,1,1]}}
 {id=Random_10, children [WavePlayer_46, WavePlayer_47, WavePlayer_48], properties {bRandomizeWithoutReplacement=false, weights [1,1,1]}}
 {id=Random_11, children [Random_9,Mixer_6], properties {bRandomizeWithoutReplacement=false, weights [1,0.4]}}
 {id=Random_12, children [WavePlayer_52, WavePlayer_53], properties {weights [1,1]}}
 {id=Random_13, children [WavePlayer_55, WavePlayer_57], properties {bRandomizeWithoutReplacement=false, weights [1,1]}}
 {id=Random_14, children [Random_12,Mixer_7], properties {bRandomizeWithoutReplacement=false, weights [1,0.4]}}
 {
  id=Random_2
  children [WavePlayer_9, WavePlayer_10, WavePlayer_11, WavePlayer_14, WavePlayer_21, WavePlayer_22]
  properties {bRandomizeWithoutReplacement=false, weights [1,1,1,1,1,1]}
 }
 {id=Random_3, children [Mixer_8,Mixer_0], properties {bRandomizeWithoutReplacement=false, weights [1,0.4]}}
 {id=Random_4, children [Random_0,Mixer_2], properties {bRandomizeWithoutReplacement=false, weights [1,0.4]}}
 {id=Random_8, children [WavePlayer_24, WavePlayer_25, WavePlayer_28], properties {bRandomizeWithoutReplacement=false, weights [1,1,1]}}
 {id=Random_9, children [WavePlayer_34, WavePlayer_35, WavePlayer_36], properties {weights [1,1,1]}}
 {id=WavePlayer_0, properties {soundWaveAssetPtr=&/PickAxe_FancyCandyMale_Swing_01}}
 {id=WavePlayer_1, properties {soundWaveAssetPtr=&/PickAxe_FancyCandyMale_Swing_02}}
 {id=WavePlayer_10, properties {soundWaveAssetPtr=&/PickAxe_FancyCandyMale_Swing_Vox_03}}
 {id=WavePlayer_11, properties {soundWaveAssetPtr=&/PickAxe_FancyCandyMale_Swing_Vox_05}}
 {id=WavePlayer_12, properties {soundWaveAssetPtr=/Game/Athena/Sounds/Weapons/Fire_Axe_Swing_Distant_02}}
 {id=WavePlayer_14, properties {soundWaveAssetPtr=&/PickAxe_FancyCandyMale_Swing_Vox_02}}
 {id=WavePlayer_21, properties {soundWaveAssetPtr=&/PickAxe_FancyCandyMale_Swing_Vox_04}}
 {id=WavePlayer_22, properties {soundWaveAssetPtr=&/PickAxe_FancyCandyMale_Swing_Vox_06}}
 {id=WavePlayer_24, properties {soundWaveAssetPtr=&/PickAxe_FancyCandyMale_Swing_Vox_01}}
 {id=WavePlayer_25, properties {soundWaveAssetPtr=&PickAxe_FancyCandyMale_Swing_Vox_03}}
 {id=WavePlayer_28, properties {soundWaveAssetPtr=&PickAxe_FancyCandyMale_Swing_Vox_05}}
 {id=WavePlayer_3, properties {soundWaveAssetPtr=&/PickAxe_FancyCandyMale_Swing_03}}
 {id=WavePlayer_34, properties {soundWaveAssetPtr=&PickAxe_FancyCandyMale_Swing_01}}
 {id=WavePlayer_35, properties {soundWaveAssetPtr=&PickAxe_FancyCandyMale_Swing_02}}
 {id=WavePlayer_36, properties {soundWaveAssetPtr=&PickAxe_FancyCandyMale_Swing_03}}
 {id=WavePlayer_4, properties {soundWaveAssetPtr=/Game/Athena/Sounds/Weapons/Fire_Axe_Swing_Distant_04}}
 {id=WavePlayer_46, properties {soundWaveAssetPtr=&PickAxe_FancyCandyMale_Swing_Vox_01}}
 {id=WavePlayer_47, properties {soundWaveAssetPtr=&PickAxe_FancyCandyMale_Swing_Vox_03}}
 {id=WavePlayer_48, properties {soundWaveAssetPtr=&PickAxe_FancyCandyMale_Swing_Vox_05}}
 {id=WavePlayer_5, properties {soundWaveAssetPtr=/Game/Athena/Sounds/Weapons/Fire_Axe_Swing_Distant_01}}
 {id=WavePlayer_52, properties {soundWaveAssetPtr=&PickAxe_FancyCandyMale_Swing_01}}
 {id=WavePlayer_53, properties {soundWaveAssetPtr=&PickAxe_FancyCandyMale_Swing_02}}
 {id=WavePlayer_55, properties {soundWaveAssetPtr=&PickAxe_FancyCandyMale_Swing_Vox_01}}
 {id=WavePlayer_57, properties {soundWaveAssetPtr=&PickAxe_FancyCandyMale_Swing_Vox_05}}
 {id=WavePlayer_6, properties {soundWaveAssetPtr=&/PickAxe_FancyCandyMale_Swing_04}}
 {id=WavePlayer_7, properties {soundWaveAssetPtr=&/PickAxe_FancyCandyMale_Swing_05}}
 {id=WavePlayer_8, properties {soundWaveAssetPtr=&/PickAxe_FancyCandyMale_Swing_06}}
 {id=WavePlayer_9, properties {soundWaveAssetPtr=&PickAxe_FancyCandyMale_Swing_Vox_01}}
]
properties {
 audioPropertiesSheet=/BRCosmetics/Sounds/Settings/PropertySheets/Pickaxe/APS_Cosmetic_Pickaxe_Swing
 bOverrideConcurrency=true
 bPrimeOnLoad=true
 maxDistance=3000
 soundClassObject=&Fort_WeaponMelee_PSM
 volumeMultiplier=0.25
 concurrencyOverrides {bLimitToOwner=true, maxCount=2, resolutionRule=StopOldest}
 concurrencySet [&Player_Melee_Weapon_Conc]
}
